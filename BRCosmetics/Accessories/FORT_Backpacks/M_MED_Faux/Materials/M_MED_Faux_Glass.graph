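# Translucent lit glass material for the M_MED_Faux backpack cosmetic, with tinted glass, a line grid and scrolling bars.
# Grid lines come from a UV1 texture, with a time-swept gradient texture driving their animated emissive glow.
# Bars pan along pre-skinned position on a selectable axis, so they follow the deforming mesh.
# Fresnel masks shape glass and darkening values; opacity takes the maximum of layered values, gated by TwoSidedSign, before the accessories override function.
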Material /BRCosmetics/Accessories/FORT_Backpacks/M_MED_Faux/Materials/M_MED_Faux_Glass
graph {
 properties {
  bUseMaterialAttributes=true
  bUsedWithMeshDeformer=true
  bUsedWithSkeletalMesh=true
  blendMode=BLEND_Translucent
  pixelDepthOffsetMode=PDOM_Legacy
  translucencyLightingMode=TLM_SurfacePerPixelLighting
  usedShadingModels=Default Lit
 }
 nodes [
  n1 = TextureSampleParameter2D(UVs=TextureCoordinate() {coordinateIndex=1}) {
   parameterName=DF_Texture
   samplerType=SAMPLERTYPE_LinearGrayscale
   texture=/BRCosmetics/Accessories/FORT_Backpacks/M_MED_Faux/Textures/Bacpack_FX_Lines
  }
  n2 = SphereMask(
   A=@n1.G
   B=ScalarParameter() {defaultValue=1, parameterName=GradientSamplePos}
   Radius=ScalarParameter() {defaultValue=0.8, parameterName=GradientLineThickness}
   Hardness=ScalarParameter() {defaultValue=0.25, parameterName=GradientLineHardness}
  )
  n3 = TextureSample(UVs=TextureCoordinate() {coordinateIndex=1}) {samplerType=SAMPLERTYPE_LinearGrayscale, texture=/BRCosmetics/Accessories/FORT_Backpacks/M_MED_Faux/Textures/Bacpack_FX_Lines_Gradient}
  n4 = Time()
  n5 = SphereMask(A=@n3.G, B=Frac(None=Multiply(A=@n4) {constB=0.25}), Radius=Constant() {r=0.25}, Hardness=Constant())
  n6 = Multiply(A=@n2, B=@n5)
  Add(A=@n2, B=@n6)
  ComponentMask(None=TextureCoordinate()) {r=true}
  n7 = NamedRerouteUsage()
  n8 = VectorParameter() {parameterName=Line_Color_02, defaultValue=<rgba 0 0 0 1>}
  n9 = VectorParameter() {parameterName=Line_Color_01, defaultValue=<rgba 0.1248 0.5274 0.9301 1>}
  n10 = PreSkinnedPosition()
  n11 = VertexInterpolator(VS=@n10)
  n12 = Add(
   A=ChannelMaskParameter(None=ShaderStageSwitch(PixelShader=@n11.PS, VertexShader=@n10)) {maskChannel=Blue, parameterName=Direction, defaultValue=<rgba 0 0 1 0>}.R
   B=Multiply(A=Time(), B=ScalarParameter() {defaultValue=1, parameterName=LinePanSpeed})
  )
  n13 = ScalarParameter() {defaultValue=0.5, parameterName=LineTiling}
  n14 = SphereMask(
   A=Frac(None=Divide(A=@n12, B=@n13))
   B=ScalarParameter() {defaultValue=1, parameterName=SamplePos}
   Radius=ScalarParameter() {defaultValue=0.5, parameterName=BarsThickness}
   Hardness=ScalarParameter() {parameterName=BarHardness}
  )
  n15 = Multiply(A=VectorParameter() {parameterName=GridShapeColor, defaultValue=<rgba 0.1248 0.5274 0.9301 1>}.RGB, B=@n2)
  n16 = ScalarParameter() {defaultValue=3, parameterName=GridShape_GlowStrength}
  n17 = Add(
   A=Multiply(A=LinearInterpolate(A=@n8.RGB, B=@n9.RGB, Alpha=@n14), B=ScalarParameter() {defaultValue=0.5, parameterName=PanningLines_Brightness})
   B=Power(Base=Multiply(A=@n15, B=@n16))
  )
  n18 = Add(
   A=LinearInterpolate(
    A=VectorParameter() {parameterName=Glass_Color_01, defaultValue=<rgba 0.0181 0.1044 0.107 1>}.RGB
    B=VectorParameter() {parameterName=Glass_Color_02, defaultValue=<rgba 0.0782 0.4508 0.4621 1>}.RGB
    Alpha=NamedRerouteUsage()
   )
   B=@n17
  )
  n19 = Max(A=Max(A=NamedRerouteUsage(), B=NamedRerouteUsage()), B=NamedRerouteUsage())
  n20 = Clamp(None=TwoSidedSign())
  n21 = TextureSampleParameter2D() {
   parameterName=Normal
   samplerType=SAMPLERTYPE_Normal
   texture=/BRCosmetics/Accessories/FORT_Backpacks/M_MED_Faux/Textures/T_M_Med_Faux_Backpack_N
  }
  n22 = SetMaterialAttributes!(
   MaterialAttributes
   Base Color=NamedRerouteDeclaration(None=Multiply(A=@n7, B=@n18)) {name=basecolor, nodeColor=<rgba 0.3594 0 1 1>}
   Metallic=NamedRerouteDeclaration(None=ScalarParameter() {parameterName=Metalness}) {name=M, nodeColor=<rgba 0 0.9766 1 1>}
   Specular=NamedRerouteDeclaration(None=ScalarParameter() {defaultValue=1, parameterName=Specular}) {name=S, nodeColor=<rgba 1 0.1406 0 1>}
   Roughness=NamedRerouteDeclaration(None=ScalarParameter() {defaultValue=0.1, parameterName=Roughness}) {name=R, nodeColor=<rgba 0 0.2734 1 1>}
   Emissive Color=NamedRerouteUsage()
   Opacity=NamedRerouteDeclaration(None=LinearInterpolate(B=@n19, Alpha=@n20)) {name=Opacity, nodeColor=<rgba 0.5 0 1 1>}
   Normal=NamedRerouteDeclaration(None=@n21.RGB) {name=N, nodeColor=<rgba 1 0 0.1875 1>}
  ) {attributeSetTypes [
    #1
    #2
    #3
    #4
    #5
    #6
    #7
   ]}
  MaterialFunctionCall(In=@n22) {materialFunction=&MF_TechArtOverrides_accessories}
  Multiply(A=@n5) {constB=5}
  Multiply(A=@n17, B=NamedRerouteUsage())
  n23 = ScalarParameter() {defaultValue=1, parameterName=FresnelExponent}
  n24 = PixelNormalWS()
  n25 = Clamp(None=Power(Base=Fresnel(ExponentIn=@n23, Normal=@n24), Exp=ScalarParameter() {defaultValue=1.2, parameterName=FresnelPower}))
  NamedRerouteDeclaration(None=@n25) {name=Glass_baseOpcityMask, nodeColor=<rgba 0.3125 1 0 1>}
  NamedRerouteDeclaration(None=LinearInterpolate(
    A=ScalarParameter() {parameterName=GridPatternMin_Opacity}
    B=ScalarParameter() {defaultValue=0.25, parameterName=GridPatternMax_Opacity}
    Alpha=@n2
   )) {name=GridPattern_Opacity, nodeColor=<rgba 1 0 0.3047 1>}
  n26 = Multiply(A=VectorParameter() {parameterName=GridShapeColor, defaultValue=<rgba 0.1248 0.5274 0.9301 1>}.RGB, B=@n6)
  n27 = ScalarParameter() {defaultValue=2, parameterName=GridAnimatedGlowStrength}
  NamedRerouteDeclaration(None=Power(Base=Multiply(A=@n26, B=@n27) {constB=2})) {name=emissive, nodeColor=<rgba 0.6875 0 1 1>}
  NamedRerouteDeclaration(None=LinearInterpolate(
    A=ScalarParameter() {defaultValue=1, parameterName=Glass_MinOpacity}
    B=ScalarParameter() {defaultValue=1, parameterName=Glass_MaxOpacity}
    Alpha=@n25
   )) {name=Glass_BaseOpacityValues, nodeColor=<rgba 1 0 0.0937 1>}
  NamedRerouteDeclaration(None=LinearInterpolate(
    A=ScalarParameter() {defaultValue=0.1, parameterName=PanningLinesmin_Opacity}
    B=ScalarParameter() {defaultValue=0.1, parameterName=PanningLinesmax_Opacity}
    Alpha=@n14
   )) {name=PanningLinesOpacityValues, nodeColor=<rgba 0.4297 0 1 1>}
  n28 = Power(
   Base=Fresnel(ExponentIn=ScalarParameter() {defaultValue=1, parameterName=Darken_FresnelExponent}, Normal=PixelNormalWS())
   Exp=ScalarParameter() {defaultValue=2, parameterName=Darken_FresnelPower}
  )
  NamedRerouteDeclaration(None=LinearInterpolate(B=ScalarParameter() {parameterName=DarkenValue}, Alpha=Clamp(None=@n28)) {constA=1}) {name=FresnelDarkenValue, nodeColor=<rgba 1 0 0.8437 1>}
 ]
}
