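# Physics asset for the M_MED_Faux backpack cosmetic, giving a plant on the pack's skeleton jointed leaf and stem bodies for secondary motion.
# Capsule bodies cover the pack, a plant master bone, and the leaf and stem bones.
# The master is angularly locked to the pack, and leaves and stems branch from it as one- or two-segment chains.
# Branch joints lock twist and swing1, leaving only swing2 for single-axis bending.
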PhysicsAsset /BRCosmetics/Accessories/FORT_Backpacks/M_MED_Faux/Meshes/M_MED_Faux_Pack_Physics
bodies {
 dyn_leaf_1_A [{shapeType=Capsule, length=2, radius=0.5, center=<0.3102 0 0>, rotation=<rot 90 0 0>}]
 dyn_leaf_1_B [{shapeType=Capsule, length=2, radius=0.5, center=<0.3102 0 0>, rotation=<rot 90 0 0>}]
 dyn_leaf_2_A [{shapeType=Capsule, length=2, radius=0.5, center=<0.3102 0 0>, rotation=<rot 90 0 0>}]
 dyn_leaf_2_B [{shapeType=Capsule, length=2, radius=0.5, center=<0.3102 0 0>, rotation=<rot 90 0 0>}]
 dyn_plant_master [{shapeType=Capsule, length=2, radius=1}]
 dyn_stem_1_A [{shapeType=Capsule, length=2, radius=0.5, center=<0.3102 0 0>, rotation=<rot 90 0 0>}]
 dyn_stem_1_B [{shapeType=Capsule, length=2, radius=0.3, center=<0.3102 0 0>, rotation=<rot 90 0 0>}]
 dyn_stem_2_A [{shapeType=Capsule, length=2, radius=0.5, center=<0.3102 0 0>, rotation=<rot 90 0 0>}]
 dyn_stem_2_B [{shapeType=Capsule, length=2, radius=0.3, center=<0.3102 0 0>, rotation=<rot 90 0 0>}]
 dyn_stem_3 [{shapeType=Capsule, length=2, radius=0.5, center=<0.3102 0 0>, rotation=<rot 90 0 0>}]
 dyn_stem_4 [{shapeType=Capsule, length=2, radius=0.5, center=<0.3102 0 0>, rotation=<rot 90 0 0>}]
 dyn_stem_5_A [{shapeType=Capsule, length=2, radius=0.5, center=<0.3102 0 0>, rotation=<rot 90 0 0>}]
 dyn_stem_5_B [{shapeType=Capsule, length=2, radius=0.3, center=<0.3102 0 0>, rotation=<rot 90 0 0>}]
 dyn_stem_6 [{shapeType=Capsule, length=2, radius=0.5, center=<0.3102 0 0>, rotation=<rot 90 0 0>}]
 pack [{shapeType=Capsule, length=2, radius=4, rotation=<rot 90 0 0>}]
}
constraints [
 {
  bone1Name=dyn_plant_master
  bone2Name=pack
  swing1LimitDegrees=45
  swing1Motion=Locked
  swing2LimitDegrees=45
  swing2Motion=Locked
  twistLimitDegrees=45
  twistMotion=Locked
 }
 {
  bone1Name=dyn_leaf_1_A
  bone2Name=dyn_plant_master
  swing1LimitDegrees=10
  swing1Motion=Locked
  swing2LimitDegrees=10
  twistLimitDegrees=45
  twistMotion=Locked
 }
 {
  bone1Name=dyn_leaf_1_B
  bone2Name=dyn_leaf_1_A
  swing1LimitDegrees=10
  swing1Motion=Locked
  swing2LimitDegrees=20
  twistLimitDegrees=45
  twistMotion=Locked
 }
 {
  bone1Name=dyn_leaf_2_A
  bone2Name=dyn_plant_master
  swing1LimitDegrees=10
  swing1Motion=Locked
  swing2LimitDegrees=10
  twistLimitDegrees=45
  twistMotion=Locked
 }
 {
  bone1Name=dyn_leaf_2_B
  bone2Name=dyn_leaf_2_A
  swing1LimitDegrees=10
  swing1Motion=Locked
  swing2LimitDegrees=20
  twistLimitDegrees=45
  twistMotion=Locked
 }
 {
  bone1Name=dyn_stem_1_A
  bone2Name=dyn_plant_master
  swing1LimitDegrees=10
  swing1Motion=Locked
  swing2LimitDegrees=10
  twistLimitDegrees=45
  twistMotion=Locked
 }
 {
  bone1Name=dyn_stem_1_B
  bone2Name=dyn_stem_1_A
  swing1LimitDegrees=10
  swing1Motion=Locked
  swing2LimitDegrees=10
  twistLimitDegrees=45
  twistMotion=Locked
 }
 {
  bone1Name=dyn_stem_2_A
  bone2Name=dyn_plant_master
  swing1LimitDegrees=10
  swing1Motion=Locked
  swing2LimitDegrees=10
  twistLimitDegrees=45
  twistMotion=Locked
 }
 {
  bone1Name=dyn_stem_2_B
  bone2Name=dyn_stem_2_A
  swing1LimitDegrees=10
  swing1Motion=Locked
  swing2LimitDegrees=10
  twistLimitDegrees=45
  twistMotion=Locked
 }
 {
  bone1Name=dyn_stem_3
  bone2Name=dyn_plant_master
  swing1LimitDegrees=10
  swing1Motion=Locked
  swing2LimitDegrees=10
  twistLimitDegrees=45
  twistMotion=Locked
 }
 {
  bone1Name=dyn_stem_4
  bone2Name=dyn_plant_master
  swing1LimitDegrees=10
  swing1Motion=Locked
  swing2LimitDegrees=10
  twistLimitDegrees=45
  twistMotion=Locked
 }
 {
  bone1Name=dyn_stem_5_A
  bone2Name=dyn_plant_master
  swing1LimitDegrees=10
  swing1Motion=Locked
  swing2LimitDegrees=10
  twistLimitDegrees=45
  twistMotion=Locked
 }
 {
  bone1Name=dyn_stem_5_B
  bone2Name=dyn_stem_5_A
  swing1LimitDegrees=10
  swing1Motion=Locked
  swing2LimitDegrees=10
  twistLimitDegrees=45
  twistMotion=Locked
 }
 {
  bone1Name=dyn_stem_6
  bone2Name=dyn_plant_master
  swing1LimitDegrees=10
  swing1Motion=Locked
  swing2LimitDegrees=10
  twistLimitDegrees=45
  twistMotion=Locked
 }
]
